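# Material function for the hero drift trail cosmetic that turns a per-axis Transform input into a world-space offset, scaling along local axes about the object pivot.
# It converts pivot-relative world position to local space for the distance along each axis.
# Each input component becomes a local axis vector, is transformed out of local space, multiplied by that distance and summed.
# Unlike X and Y, Z takes its distance from a dot product with ObjectOrientation.
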MaterialFunction /VehicleCosmetics/DriftTrail/DriftTrail_Hero/Material/MF_ObjectScale_Local_Pivot
graph {nodes [
  n1 = FunctionInput() {bUsePreviewValueAsDefault=true, inputName=Transform}
  n2 = Constant()
  n3 = AppendVector(A=ComponentMask(None=@n1) {r=true}, B=@n2)
  n4 = Transform(None=Subtract(
    A=WorldPosition().XYZ
    B=MaterialFunctionCall() {materialFunction=&ObjectPivotPoint}.Object Pivot Location
   )) {transformSourceType=TRANSFORMSOURCE_World, transformType=TRANSFORM_Local}
  n5 = Multiply(A=Transform(None=AppendVector(A=@n3, B=@n2)) {transformSourceType=TRANSFORMSOURCE_Local}, B=ComponentMask(None=@n4) {r=true})
  n6 = AppendVector(A=ComponentMask(None=@n1) {g=true}, B=@n2)
  n7 = Multiply(A=Transform(None=AppendVector(A=@n6, B=@n2)) {transformSourceType=TRANSFORMSOURCE_Local}, B=ComponentMask(None=@n4) {g=true})
  n8 = Transform(None=AppendVector(A=AppendVector(A=@n2, B=@n2), B=ComponentMask(None=@n1) {b=true})) {transformSourceType=TRANSFORMSOURCE_Local}
  n9 = DotProduct(A=@n4, B=ObjectOrientation())
  FunctionOutput(None=Add(A=Add(A=@n5, B=@n7), B=Multiply(A=@n8, B=@n9)))
 ]}
