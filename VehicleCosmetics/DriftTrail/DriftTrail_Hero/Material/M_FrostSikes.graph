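# M_FrostSikes is an unlit, masked Niagara mesh-particle material for ice spikes in the Hero drift trail.
# Ice color layers bump-offset noise and detail normals, Fresnel tints, cracks, and a fake reflection sampled by reflection vector.
# Dynamic parameters feed a local-X gradient into MF_ObjectScale_Local_Pivot, but both lerp inputs read Min, leaving Max unused.
# The graph holds two copies of the ice and scale networks, with no material output connections shown.
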Material /VehicleCosmetics/DriftTrail/DriftTrail_Hero/Material/M_FrostSikes
graph {
 properties {
  bAutomaticallySetUsageInEditor=false
  bUsedWithNiagaraMeshParticles=true
  blendMode=BLEND_Masked
  shadingModel=MSM_Unlit
  usedShadingModels=Unlit
 }
 nodes [
  MaterialExpressionMeshEmitterDynamicParameter_0 = DynamicParameter() {paramNames [Falloff,Min,Max,Shift]}
  MaterialExpressionObjectWorldPosition_0 = ObjectPositionWS() {originType=CameraRelative}
  n1 = WorldPosition()
  n2 = Add(
   A=Transform(None=Subtract(A=@MaterialExpressionObjectWorldPosition_0, B=@n1.XYZ)) {transformSourceType=TRANSFORMSOURCE_Local}
   B=@MaterialExpressionMeshEmitterDynamicParameter_0.Shift
  )
  MaterialExpressionConstantClamp_19 = Saturate(None=Multiply(A=ComponentMask(None=@n2) {r=true}, B=@MaterialExpressionMeshEmitterDynamicParameter_0.Falloff))
  MaterialFunctionCall!(Transform=LinearInterpolate(
    A=@MaterialExpressionMeshEmitterDynamicParameter_0.Min
    B=@MaterialExpressionMeshEmitterDynamicParameter_0.Min
    Alpha=@MaterialExpressionConstantClamp_19
   )) {materialFunction=&/MF_ObjectScale_Local_Pivot}
  n3 = DynamicParameter() {paramNames [Falloff,Min,Max,Shift]}
  n4 = Transform(None=Subtract(A=ObjectPositionWS() {originType=CameraRelative}, B=WorldPosition().XYZ)) {transformSourceType=TRANSFORMSOURCE_Local}
  n5 = Multiply(A=ComponentMask(None=Add(A=@n4, B=@n3.Shift)) {r=true}, B=@n3.Falloff)
  MaterialFunctionCall(Transform=LinearInterpolate(A=@n3.Min, B=@n3.Min, Alpha=Saturate(None=@n5))) {materialFunction=&MF_ObjectScale_Local_Pivot}
  n6 = BumpOffset(
   Coordinate=Multiply(A=TextureCoordinate() {uTiling=4, vTiling=4}, B=ScalarParameter() {defaultValue=1, group=Ice, parameterName=IceTile})
   Height=ScalarParameter() {defaultValue=1, group=Ice, parameterName=BumpHeight}
  ) {heightRatio=0.15}
  n7 = ConstantBiasScale(None=TextureSample(UVs=@n6) {samplerType=SAMPLERTYPE_Normal, texture=/DealershipSharedLibrary/FX/Textures/Normal/T_VCVFX_NoiseStone_01_N}.RGB) {scale=0.2}
  n8 = TextureSampleParameter2D(UVs=@n6) {parameterName=IceDetail_Normal, samplerType=SAMPLERTYPE_Normal, texture=/DealershipSharedLibrary/FX/Textures/Normal/T_VCVFX_Ice_01_N}
  n9 = Add(A=@n7, B=@n8.RGB)
  n10 = Normalize(VectorInput=Multiply(A=@n9, B=VectorParameter() {group=Ice, parameterName=Ice_Normal_Int, defaultValue=<rgba 1 1 1 1>}.RGB))
  n11 = Multiply(A=MaterialFunctionCall(Normal Vector=@n10, Camera Vector=NamedRerouteUsage(), Invert Fresnel=StaticBool(), Power=Constant() {r=5.5}) {materialFunction=&Fresnel_Function}.Result) {constB=20}
  n12 = ReflectionVectorWS()
  n13 = ComponentMask(None=@n12) {g=true, r=true}
  n14 = Multiply(
   A=Multiply(
    A=Saturate(None=@n11)
    B=TextureSampleParameter2D(UVs=@n13) {group=Ice, parameterName=reflection, texture=/DealershipSharedLibrary/FX/Textures/Noise/T_VCVFX_Water01_A}.R
   )
   B=ScalarParameter() {defaultValue=1, group=Ice, parameterName=Reflection_Int}
  )
  n15 = Power(Base=DotProduct(A=NamedRerouteUsage(), B=@n7), Exp=Constant() {r=3})
  n16 = VectorParameter() {group=Ice, parameterName=OuterFresnelColor, defaultValue=<rgba 0.3101 0.625 0.9944 1>}
  n17 = VectorParameter() {group=Ice, parameterName=InnerFresnelColor, defaultValue=<rgba 0.2028 0.3001 0.7016 1>}
  n18 = TextureSample(UVs=TextureCoordinate() {uTiling=2, vTiling=2}) {texture=/DealershipSharedLibrary/FX/Textures/Diffuse/T_VCVFX_Ice_D}
  n19 = ScalarParameter() {defaultValue=5, group=Ice, parameterName=IceCracksIntensity}
  n20 = Constant4Vector() {constant=<rgba 0 1 0 0.1>}
  n21 = MaterialFunctionCall(
   Input=MaterialFunctionCall(
    Normal Vector=Normalize(VectorInput=@n9)
    Camera Vector=NamedRerouteUsage()
    Invert Fresnel=StaticBool() {value=true}
    Power=Constant() {r=3}
   ) {materialFunction=&Fresnel_Function}.Result
   Input Low=@n20.R
   Input High=@n20.R
   Target Low=@n20.R
   Target High=@n20.R
  ) {materialFunction=&RemapValueRange}
  n22 = Constant() {r=2}
  n23 = Saturate(None=Add(A=Multiply(A=@n18.R, B=@n19), B=Multiply(A=@n21.Result, B=@n22)))
  n24 = ScalarParameter() {defaultValue=0.1, group=Ice, parameterName=Normal_Fresnel_Int}
  n25 = Constant4Vector() {constant=<rgba 0 1 0 0.1>}
  n26 = MaterialFunctionCall(
   Input=MaterialFunctionCall(Normal Vector=@n10, Camera Vector=NamedRerouteUsage(), Invert Fresnel=StaticBool() {value=true}, Power=Constant() {r=6}) {materialFunction=&Fresnel_Function}.Result
   Input Low=@n25.R
   Input High=@n25.R
   Target Low=@n25.R
   Target High=@n25.R
  ) {materialFunction=&RemapValueRange}
  n27 = Add(
   A=Add(A=Saturate(None=@n14), B=Multiply(A=@n15, B=@n16.RGB))
   B=Add(A=LinearInterpolate(A=@n17.RGB, B=@n16.RGB, Alpha=@n23), B=Multiply(A=@n24, B=@n26.Result))
  )
  n28 = ScalarParameter() {defaultValue=0.6, parameterName=Ice_Overall_Int}
  MaterialFunctionCall(Emissive=Saturate(None=Multiply(A=@n27, B=@n28))) {materialFunction=&MF_FX_EmissiveAdjustments}
  n29 = BumpOffset(
   Coordinate=Multiply(A=TextureCoordinate() {uTiling=4, vTiling=4}, B=ScalarParameter() {defaultValue=1, group=Ice, parameterName=IceTile})
   Height=ScalarParameter() {defaultValue=1, group=Ice, parameterName=BumpHeight}
  ) {heightRatio=0.15}
  n30 = ConstantBiasScale(None=TextureSample(UVs=@n29) {samplerType=SAMPLERTYPE_LinearColor, texture=&T_VCVFX_NoiseStone_01_N}.RGB) {scale=0.2}
  n31 = TextureSampleParameter2D(UVs=@n29) {
   group=Ice
   parameterName=IceDetail_Normal
   samplerType=SAMPLERTYPE_LinearColor
   texture=&T_VCVFX_Ice_01_N
  }
  n32 = Add(A=@n30, B=@n31.RGB)
  n33 = Multiply(A=@n32, B=VectorParameter() {group=Ice, parameterName=Ice_Normal_Int, defaultValue=<rgba 1 1 1 1>}.RGB)
  n34 = Fresnel(Normal=@n33) {exponent=1}
  n35 = TextureSampleParameter2D(UVs=ComponentMask(None=ReflectionVectorWS()) {g=true, r=true}) {group=Ice, parameterName=reflection, texture=&T_VCVFX_Water01_A}
  MaterialExpressionConstantClamp_22 = Saturate(None=Multiply(A=Multiply(A=@n34, B=@n35.R), B=ScalarParameter() {defaultValue=1, group=Ice, parameterName=Reflection_Int}))
  n36 = Power(Base=DotProduct(A=CameraVectorWS(), B=@n30), Exp=Constant() {r=3})
  n37 = VectorParameter() {group=Ice, parameterName=OuterFresnelColor, defaultValue=<rgba 0.3101 0.625 0.9944 1>}
  n38 = VectorParameter() {group=Ice, parameterName=InnerFresnelColor, defaultValue=<rgba 0.2028 0.3001 0.7016 1>}
  n39 = TextureCoordinate() {uTiling=2, vTiling=2}
  n40 = Add(
   A=Multiply(
    A=TextureSample(UVs=@n39) {texture=&T_VCVFX_Ice_D}.R
    B=ScalarParameter() {defaultValue=5, group=Ice, parameterName=IceCracksIntensity}
   )
   B=Multiply(A=Fresnel(Normal=@n32) {exponent=3}, B=Constant() {r=0.2})
  )
  n41 = ScalarParameter() {defaultValue=0.1, group=Ice, parameterName=Normal_Fresnel_Int}
  n42 = Fresnel(Normal=@n33) {exponent=6}
  n43 = Add(
   A=Add(A=@MaterialExpressionConstantClamp_22, B=Multiply(A=@n36, B=@n37.RGB))
   B=Add(A=LinearInterpolate(A=@n38.RGB, B=@n37.RGB, Alpha=@n40), B=Multiply(A=@n41, B=@n42))
  )
  n44 = ScalarParameter() {defaultValue=0.6, parameterName=Ice_Overall_Int}
  Multiply(A=ParticleColor().RGB, B=Saturate(None=Multiply(A=@n43, B=@n44)))
  NamedRerouteDeclaration(None=Transform(None=CameraVectorWS()) {transformSourceType=TRANSFORMSOURCE_World, transformType=TRANSFORM_Tangent}) {name=Camera Vector, nodeColor=<rgba 0 1 0.7422 1>}
  ParticleColor()
  Transform(None=@n12) {transformSourceType=TRANSFORMSOURCE_World, transformType=TRANSFORM_Tangent}
 ]
}
