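# Translucent unlit Niagara ribbon material for the Hero drift trail vehicle cosmetic, with two looks chosen by the MainTrail static switch.
# When on, panning ice and cloud textures build a snowy emissive, floored masks cut a hard-edged opacity, and sine/cosine offsets wave the ribbon.
# When off, particle-colored, blue-tinted streak layers with scrolling noise blend along the ribbon's U axis with a white, depth-faded fire-trail layer.
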Material /VehicleCosmetics/DriftTrail/DriftTrail_Hero/Material/M_HeroTrailMixed
graph {
 properties {
  bAutomaticallySetUsageInEditor=false
  bUseMaterialAttributes=true
  bUsedWithNiagaraRibbons=true
  blendMode=BLEND_Translucent
  shadingModel=MSM_Unlit
  usedShadingModels=Unlit
 }
 nodes [
  Constant4Vector() {constant=<rgba 1 1 1 0.5>}
  n1 = OneMinus(None=ComponentMask(None=TextureCoordinate()) {r=true})
  MaterialFunctionCall!(Base Material, Top Material, Alpha=Ceil(None=Power(Base=@n1) {constExponent=1})) {materialFunction=/Engine/Functions/MaterialLayerFunctions/MatLayerBlend_Standard}
  n2 = Transform(None=ReflectionVectorWS()) {transformSourceType=TRANSFORMSOURCE_Local}
  n3 = Constant3Vector() {constant=<rgba 0.0452 0.1878 0.521 1>}
  MaterialExpressionConstantClamp_10 = Saturate(None=ConstantBiasScale(None=Multiply(A=@n2, B=@n3)) {bias=0.1, scale=0.2})
  Multiply(
   A=ScalarParameter() {defaultValue=4, parameterName=ReflectionAmount}
   B=Desaturation(None=@MaterialExpressionConstantClamp_10, Fraction=Constant()) {luminanceFactors=<rgba 0 0 0.11 0>}
  )
  n4 = Panner(Coordinate=TextureCoordinate() {uTiling=8, vTiling=2}) {speedX=-2}
  n5 = Multiply(
   A=OneMinus(None=TextureSample(UVs=@n4) {texture=/DealershipSharedLibrary/FX/Textures/TexturePackage/T_VCVFX_IceT_RGB}.B)
   B=Constant() {r=0.3}
  )
  MaterialExpressionVertexColor_1 = ParticleColor()
  n6 = Multiply(A=Constant3Vector() {constant=<rgba 0.8 1 1 1>}, B=@MaterialExpressionVertexColor_1.RGB)
  MaterialExpressionConstantClamp_9 = Saturate(None=Multiply(A=Multiply(A=@n3, B=@n6) {constA=1}, B=@MaterialExpressionVertexColor_1.A))
  n7 = TextureSampleParameter2D(UVs=Panner(Coordinate=TextureCoordinate() {uTiling=4}) {speedX=-2}) {parameterName=DetailTexture, texture=/DealershipSharedLibrary/FX/Textures/Noise/T_VCVFX_CloudsSoft_RGB}
  n8 = ScalarParameter() {defaultValue=-0.1, parameterName=MaskAmount}
  MaterialExpressionConstantClamp_11 = Saturate(None=Multiply(A=Add(A=@n7.G, B=@n8), B=ScalarParameter() {defaultValue=0.5, parameterName=Sharpness}))
  n9 = LinearInterpolate(
   A=@MaterialExpressionConstantClamp_9
   B=VectorParameter() {parameterName=Snow, defaultValue=<rgba 1 1 1 1>}.RGB
   Alpha=Multiply(A=@MaterialExpressionConstantClamp_11, B=ScalarParameter() {defaultValue=0.6, parameterName=MaskBrightness})
  )
  n10 = Panner(Coordinate=TextureCoordinate() {uTiling=4}) {speedX=-1}
  n11 = Multiply(
   A=Add(
    A=TextureSample(UVs=@n10) {texture=&T_VCVFX_IceT_RGB}.R
    B=TextureSample() {texture=&T_VCVFX_IceT_RGB}.G
   )
   B=Constant() {r=0.6}
  )
  n12 = Constant3Vector() {constant=<rgba 0.6 0.7 2 0>}
  MaterialExpressionConstantClamp_8 = Saturate(None=Add(A=Add(A=@n5, B=@n9), B=Multiply(A=@n11, B=@n12)))
  n13 = ScalarParameter() {defaultValue=1.5, parameterName=Opacity}
  n14 = ComponentMask(None=TextureCoordinate() {coordinateIndex=1}) {r=true}
  n15 = Subtract(A=OneMinus(None=@n14)) {constB=0.45}
  n16 = Multiply(A=TextureSample(UVs=@n10) {texture=/DealershipSharedLibrary/FX/Textures/Diffuse/T_VCVFX_mgm_D}.A, B=Multiply(A=@n15, B=Constant() {r=3}))
  n17 = Multiply(
   A=Multiply(A=@n15, B=Constant() {r=5})
   B=TextureSample(UVs=@n10) {texture=/DealershipSharedLibrary/FX/Textures/Gradient_Masks/T_VCVFX_mgm_A}.RGB
  )
  n18 = Add(A=TextureCoordinate() {uTiling=0.65, vTiling=0.25}, B=Constant2Vector() {g=-0.13, r=-0.447})
  n19 = OneMinus(None=DotProduct(A=@n18, B=@n18))
  n20 = Constant() {r=0.8}
  MaterialExpressionConstantClamp_1 = Saturate(None=Multiply(A=Subtract(A=@n19, B=@n20), B=Constant() {r=100}))
  MaterialExpressionConstantClamp_2 = Saturate(None=Multiply(A=@n14, B=Constant() {r=15}))
  n21 = Constant() {r=1}
  n22 = Multiply(
   A=Power(Base=Add(A=@n16, B=@n17), Exp=Constant() {r=1})
   B=Multiply(A=@MaterialExpressionConstantClamp_1, B=Power(Base=@MaterialExpressionConstantClamp_2, Exp=@n21))
  )
  n23 = Power(Base=Multiply(A=@n15, B=Constant() {r=4}), Exp=Constant())
  n24 = ComponentMask(None=TextureCoordinate() {coordinateIndex=1}) {g=true}
  n25 = Multiply(A=@n24, B=OneMinus(None=@n24))
  n26 = Constant() {r=4}
  n27 = Multiply(A=@MaterialExpressionConstantClamp_1, B=Power(Base=Multiply(A=@n25, B=@n26), Exp=Constant() {r=3}))
  n28 = Panner(Coordinate=TextureCoordinate() {vTiling=0.5}) {speedX=-3, speedY=3}
  n29 = Multiply(
   A=Multiply(A=Multiply(A=@n23, B=@n27), B=TextureSample(UVs=@n28) {texture=&T_VCVFX_CloudsSoft_RGB}.G)
   B=Multiply(A=@n15, B=Constant() {r=12})
  )
  MaterialExpressionConstantClamp_0 = Saturate(None=Floor(None=Add(A=@n22, B=@n29)))
  MaterialExpressionConstantClamp_13 = Saturate(None=Multiply(A=Add(A=@n13, B=@n7.G), B=@MaterialExpressionConstantClamp_0))
  n30 = OneMinus(None=Time())
  n31 = ScalarParameter() {defaultValue=0.5, group=WPO, parameterName=Speed}
  n32 = ComponentMask(None=TextureCoordinate()) {r=true}
  n33 = ScalarParameter() {defaultValue=1, group=WPO, parameterName=tile}
  n34 = Cosine(None=Add(A=Multiply(A=@n30, B=@n31), B=Divide(A=@n32, B=@n33))) {period=0.7}
  n35 = Sine(None=Add()) {period=0.2}
  n36 = Multiply(A=AppendVector(A=AppendVector(A=@n34, B=@n35), B=Constant()), B=ScalarParameter() {defaultValue=0.5, group=WPO, parameterName=Intensity})
  n37 = ScalarParameter() {defaultValue=1, group=WPO, parameterName=Amount}
  n38 = ComponentMask(None=TextureCoordinate()) {r=true}
  n39 = ScalarParameter() {parameterName=PowerMiddle2}
  n40 = Multiply(
   A=ConstantBiasScale(None=Multiply(A=@n36, B=@n37))
   B=Subtract(A=Power(Base=@n38, Exp=@n39), B=ScalarParameter() {defaultValue=0.05, parameterName=StaticPoint})
  )
  MaterialExpressionVertexColor_6 = ParticleColor()
  n41 = Constant3Vector() {constant=<rgba 0.4 0.6 1 1>}
  n42 = Panner(Coordinate=TextureCoordinate() {uTiling=2}) {speedX=-1}
  n43 = ComponentMask(None=TextureCoordinate()) {r=true}
  n44 = OneMinus(None=@n43)
  n45 = Multiply(
   A=Multiply(A=@MaterialExpressionVertexColor_6.RGB, B=Multiply(B=@n41) {constA=20})
   B=Multiply(A=TextureSample(UVs=@n42) {texture=/DealershipSharedLibrary/FX/Textures/Diffuse/T_VCVFX_Details_Diffuse}.R, B=Subtract(A=@n44) {constB=0.4})
  )
  n46 = Constant() {r=2}
  n47 = ComponentMask(None=TextureCoordinate()) {r=true}
  n48 = Multiply(A=Power(Base=OneMinus(None=@n47), Exp=Constant() {r=3}), B=Constant() {r=3})
  n49 = Multiply(A=ScalarParameter() {defaultValue=1, parameterName=MaskSpeed}, B=Time())
  n50 = TextureSample(UVs=Panner(Coordinate=TextureCoordinate() {vTiling=0.4}, Time=@n49) {speedX=-1, speedY=-0.1}) {texture=/VehicleCosmetics/FX/Textures/Noise/T_VCVFX_NoiseStarBeam_A}
  n51 = Constant() {r=1.5}
  n52 = TextureSample(UVs=Panner(Coordinate=TextureCoordinate() {uTiling=0.8, vTiling=0.5}, Time=@n49) {speedX=-1, speedY=-0.1}) {texture=&T_VCVFX_NoiseStarBeam_A}
  n53 = Constant() {r=1.5}
  MaterialExpressionConstantClamp_18 = Saturate(None=Add(A=Multiply(A=@n50.RGB, B=@n51), B=Multiply(A=@n52.RGB, B=@n53)))
  n54 = Multiply(A=@n48, B=@MaterialExpressionConstantClamp_18)
  n55 = Multiply(A=@n45, B=Multiply(A=Multiply(A=@n43, B=@n46), B=Add(A=@n48, B=@n54)))
  n56 = TextureSample() {texture=/DealershipSharedLibrary/FX/Textures/Gradient_Masks/T_VCVFX_MainTrailDiffuse_A}
  MaterialExpressionVertexColor_5 = ParticleColor()
  n57 = Multiply(A=Multiply(B=Constant3Vector() {constant=<rgba 0.4 0.6 1 1>}) {constA=6}, B=@MaterialExpressionVertexColor_5.RGB)
  n58 = Multiply(A=@n56.RGB, B=@n57)
  n59 = TextureSample(UVs=Panner(Coordinate=TextureCoordinate() {uTiling=1.8}) {speedX=-2.5}) {texture=&T_VCVFX_Details_Diffuse}
  n60 = LinearInterpolate(A=Constant() {r=0.75}, B=Constant() {r=0.5}, Alpha=Sine(None=Time()) {period=2})
  n61 = Add(A=Multiply(A=@n57, B=@n59.G), B=Multiply(A=Multiply(A=@n60, B=@n59.R), B=@n57))
  n62 = Constant() {r=2}
  n63 = Multiply(A=@n56.RGB, B=@MaterialExpressionVertexColor_5.A)
  n64 = Multiply(A=@n61, B=@n57)
  n65 = Multiply(A=ComponentMask(None=TextureCoordinate()) {r=true}, B=Constant() {r=2})
  n66 = OneMinus(None=ComponentMask(None=TextureCoordinate()) {r=true})
  n67 = Constant() {r=3}
  n68 = Multiply(A=ScalarParameter() {defaultValue=1, parameterName=MaskSpeed}, B=Time())
  n69 = TextureSample(UVs=Panner(Coordinate=TextureCoordinate() {vTiling=0.4}, Time=@n68) {speedX=-1, speedY=-0.1}) {texture=&T_VCVFX_NoiseStarBeam_A}
  n70 = Constant() {r=1.5}
  n71 = TextureSample(UVs=Panner(Coordinate=TextureCoordinate() {uTiling=0.8, vTiling=0.5}, Time=@n68) {speedX=-1, speedY=-0.1}) {texture=&T_VCVFX_NoiseStarBeam_A}
  n72 = Constant() {r=1.5}
  MaterialExpressionConstantClamp_17 = Saturate(None=Add(A=Multiply(A=@n69.RGB, B=@n70), B=Multiply(A=@n71.RGB, B=@n72)))
  n73 = Add(A=Multiply(A=Power(Base=@n66, Exp=@n67), B=Constant() {r=3}), B=Multiply(B=@MaterialExpressionConstantClamp_17))
  MaterialExpressionConstantClamp_16 = Saturate(None=Multiply(A=LinearInterpolate(A=@n63, B=@n64, Alpha=@n62), B=Multiply(A=@n65, B=@n73)))
  n74 = ComponentMask(None=TextureCoordinate()) {r=true}
  n75 = MaterialFunctionCall(
   Base Material=SetMaterialAttributes!(MaterialAttributes, Emissive Color=@n45, Opacity=Saturate(None=@n55)) {attributeSetTypes [#1, #2]}
   Top Material=SetMaterialAttributes(Emissive Color=LinearInterpolate(A=@n58, B=@n61, Alpha=@n62), Opacity=@MaterialExpressionConstantClamp_16) {attributeSetTypes [#1, #2]}
   Alpha=Subtract(A=OneMinus(None=@n74)) {constB=0.4}
  ) {materialFunction=&MatLayerBlend_Standard}
  n76 = TextureSample() {texture=&T_VCVFX_MainTrailDiffuse_A}
  n77 = ParticleColor()
  n78 = Multiply(A=Multiply(B=Constant3Vector() {constant=<rgba 0.4 0.6 1 1>}) {constA=6}, B=@n77.RGB)
  n79 = TextureSample(UVs=Panner(Coordinate=TextureCoordinate() {uTiling=1.8}) {speedX=-2.5}) {texture=&T_VCVFX_Details_Diffuse}
  n80 = LinearInterpolate(A=Constant() {r=0.75}, B=Constant() {r=0.5}, Alpha=Sine(None=Time()) {period=2})
  n81 = Add(A=Multiply(A=@n78, B=@n79.G), B=Multiply(A=Multiply(A=@n80, B=@n79.R), B=@n78))
  n82 = Constant()
  n83 = LinearInterpolate(A=Multiply(A=@n76.RGB, B=@n77.A), B=Multiply(A=@n81, B=@n78), Alpha=@n82)
  n84 = TextureCoordinate()
  n85 = TextureCoordinate()
  n86 = Power(Base=OneMinus(None=ComponentMask(None=@n85) {r=true}), Exp=Constant() {r=3})
  n87 = Constant() {r=3}
  n88 = Multiply(A=ScalarParameter() {defaultValue=1, parameterName=MaskSpeed}, B=Time())
  n89 = TextureSample(UVs=Panner(Coordinate=TextureCoordinate() {vTiling=0.4}, Time=@n88) {speedX=-1, speedY=-0.1}) {texture=&T_VCVFX_NoiseStarBeam_A}
  n90 = Constant() {r=1.5}
  n91 = TextureSample(UVs=Panner(Coordinate=TextureCoordinate() {uTiling=0.8, vTiling=0.5}, Time=@n88) {speedX=-1, speedY=-0.1}) {texture=&T_VCVFX_NoiseStarBeam_A}
  n92 = Constant() {r=1.5}
  n93 = Saturate(None=Add(A=Multiply(A=@n89.RGB, B=@n90), B=Multiply(A=@n91.RGB, B=@n92)))
  n94 = Multiply(A=Multiply(A=ComponentMask(None=@n84) {r=true}, B=Constant() {r=2}), B=Add(A=Multiply(A=@n86, B=@n87), B=Multiply(B=@n93)))
  n95 = SetMaterialAttributes(Emissive Color=LinearInterpolate(A=Multiply(A=@n76.RGB, B=@n78), B=@n81, Alpha=@n82), Opacity=Saturate(None=Multiply(A=@n83, B=@n94))) {attributeSetTypes [#1, #2]}
  n96 = TextureCoordinate()
  n97 = Subtract(A=OneMinus(None=ComponentMask(None=@n96) {r=true})) {constB=0.4}
  MaterialExpressionVertexColor_2 = ParticleColor()
  n98 = Time()
  n99 = Constant() {r=0.2}
  n100 = TextureSample(UVs=Panner(Coordinate=TextureCoordinate(), Time=Multiply(A=@n98, B=@n99)) {speedX=-4}) {texture=/DealershipSharedLibrary/FX/Textures/TexturePackage/T_VCVFX_FireTrail_RGB}
  n101 = StaticSwitchParameter(True=@n100.R, False=@n100.R) {parameterName=Smoke}
  n102 = TextureSample() {texture=/DealershipSharedLibrary/FX/Textures/Gradient_Masks/T_VCVFX_SingularityTrailTransparency_A}
  n103 = Constant() {r=1.3}
  MaterialExpressionConstantClamp_14 = Saturate(None=Multiply(A=Multiply(A=@MaterialExpressionVertexColor_2.A, B=@n101), B=Power(Base=@n102.RGB, Exp=@n103)))
  n104 = Multiply(A=Constant3Vector() {constant=<rgba 1 1 1 0>}, B=@MaterialExpressionConstantClamp_14)
  n105 = Constant() {r=1}
  n106 = Multiply(A=Multiply(
    A=@MaterialExpressionConstantClamp_14
    B=MaterialFunctionCall(Opacity In=@n105) {materialFunction=&MF_VCVFX_DepthFade}.Result
   )) {constB=0.3}
  n107 = OneMinus(None=ComponentMask(None=TextureCoordinate()) {r=true})
  StaticSwitchParameter(
   True=SetMaterialAttributes!(
    MaterialAttributes
    Emissive Color=MaterialFunctionCall(Emissive=@MaterialExpressionConstantClamp_8) {materialFunction=&MF_FX_EmissiveAdjustments}.Result
    Opacity=@MaterialExpressionConstantClamp_13
    World Position Offset=Transform(None=@n40)
   ) {attributeSetTypes [#1, #2, #3]}
   False=MaterialFunctionCall(
    Base Material=MaterialFunctionCall(Base Material=@n75.Blended Material, Top Material=@n95, Alpha=@n97) {materialFunction=&MatLayerBlend_Standard}.Blended Material
    Top Material=SetMaterialAttributes!(MaterialAttributes, Emissive Color=@n104, Opacity=@n106) {attributeSetTypes [#1, #2]}
    Alpha=Subtract(A=@n107) {constB=0.4}
   ) {materialFunction=&MatLayerBlend_Standard}.Blended Material
  ) {parameterName="MainTrail>"}
 ]
}
