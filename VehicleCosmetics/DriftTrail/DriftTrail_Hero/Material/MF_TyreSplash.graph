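# Material function for the tyre splash in the Hero drift trail VFX, outputting emissive and opacity for particles.
# Emissive multiplies particle colour by a panning splash texture plus boosted foam-gradient layers, with a dynamic Offset shifting UVs along U.
# Opacity combines particle alpha, a static-switched splash or noise source, and a splash mask modulated by scrolling noise and dynamic Erode.
# A U-axis edge falloff and camera depth fade soften the result.
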MaterialFunction /VehicleCosmetics/DriftTrail/DriftTrail_Hero/Material/MF_TyreSplash
graph {nodes [
  MaterialExpressionVertexColor_2 = ParticleColor()
  n1 = TextureCoordinate()
  n2 = DynamicParameter() {paramNames [Erode,Offset,Param3,Param4]}
  n3 = AppendVector(A=@n2.Offset, B=Constant())
  n4 = Time()
  n5 = ScalarParameter() {defaultValue=1, parameterName=MainSpeed}
  n6 = TextureSample(UVs=Panner(Coordinate=Add(A=@n1, B=@n3), Time=Multiply(A=@n4, B=@n5)) {speedX=1}) {texture=/VehicleCosmetics/DriftTrail/DriftTrail_Hero/Textures/T_VCVFX_DriftTrail_Hero_WWrapSplash_RGB}
  n7 = TextureSample(UVs=TextureCoordinate()) {texture=/DealershipSharedLibrary/FX/Textures/TexturePackage/T_VCVFX_TireFoamGradient01_RGB}
  n8 = Constant() {r=3}
  n9 = Constant()
  n10 = Panner(Coordinate=Add(A=TextureCoordinate() {uTiling=2, vTiling=3}, B=AppendVector(A=@n2.Offset, B=@n9))) {speedX=0.5, speedY=-1.5}
  n11 = Multiply(
   A=Multiply(
    A=Multiply(A=@n7.G, B=@n8)
    B=TextureSample(UVs=@n10) {texture=&T_VCVFX_TireFoamGradient01_RGB}.G
   )
   B=Constant() {r=18}
  )
  FunctionOutput(None=Multiply(A=@MaterialExpressionVertexColor_2.RGB, B=Add(A=@n6.R, B=@n11))) {outputName=Emissive}
  MaterialExpressionVertexColor_3 = ParticleColor()
  n12 = Add(A=Multiply(A=@n6.G, B=Constant3Vector() {constant=<rgba 1 1 1 0>}), B=@n6.R)
  n13 = ScalarParameter() {defaultValue=1, parameterName=NoiseSpeed}
  n14 = TextureSample(UVs=Panner(Coordinate=TextureCoordinate() {uTiling=1.5, vTiling=1.5}, Time=Multiply(A=@n4, B=@n13)) {speedY=-2.3}) {texture=&T_VCVFX_TireFoamGradient01_RGB}
  n15 = Subtract(A=@n14.G, B=@n2.Erode)
  MaterialExpressionConstantClamp_1 = Saturate(None=Multiply(A=@n6.R, B=OneMinus(None=@n15)))
  n16 = Multiply(
   A=@MaterialExpressionVertexColor_3.A
   B=Multiply(A=StaticSwitchParameter(True=@n12, False=@n14.G) {defaultValue=true, parameterName=Noise}, B=@MaterialExpressionConstantClamp_1)
  )
  n17 = TextureCoordinate()
  n18 = ComponentMask(None=@n17) {r=true}
  n19 = Multiply(A=Multiply(A=ComponentMask(None=@n17) {r=true}, B=Constant() {r=1}), B=Multiply(A=OneMinus(None=@n18), B=Constant() {r=1}))
  n20 = Constant() {r=40}
  n21 = Constant() {r=75}
  FunctionOutput(None=Multiply(
    A=Multiply(A=@n16, B=@n19)
    B=MaterialFunctionCall(Fade Length=@n20, Fade Offset=@n21) {materialFunction=&CameraDepthFade}.Result
   )) {outputName=Opcity, sortPriority=1}
 ]}
