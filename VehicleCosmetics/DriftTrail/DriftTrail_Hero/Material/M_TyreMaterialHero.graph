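# Masked, unlit, two-sided material for Niagara mesh particles, rendering an icy, snowy tyre effect for the Hero drift trail.
# Panning ice textures and particle-coloured blue blend toward a Snow colour through a noise mask.
# Panned spikes, clouds and UV1 falloff are floored into a hard-edged shape.
# MF_TyreSplash layers over this before emissive adjustment, and its opacity joins the shape mask, banded by a rounded sine.
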Material /VehicleCosmetics/DriftTrail/DriftTrail_Hero/Material/M_TyreMaterialHero
graph {
 properties {
  bAutomaticallySetUsageInEditor=false
  bUsedWithNiagaraMeshParticles=true
  blendMode=BLEND_Masked
  shadingModel=MSM_Unlit
  twoSided=true
  usedShadingModels=Unlit
 }
 nodes [
  Constant() {r=1}
  n1 = Panner(Coordinate=TextureCoordinate() {uTiling=8, vTiling=2}) {speedX=2}
  n2 = Multiply(
   A=OneMinus(None=TextureSample(UVs=@n1) {texture=/DealershipSharedLibrary/FX/Textures/TexturePackage/T_VCVFX_IceT_RGB}.B)
   B=Constant() {r=0.3}
  )
  n3 = Constant3Vector() {constant=<rgba 0.0452 0.1878 0.521 1>}
  MaterialExpressionVertexColor_1 = ParticleColor()
  n4 = Multiply(A=Constant3Vector() {constant=<rgba 0.8 1 1 1>}, B=@MaterialExpressionVertexColor_1.RGB)
  MaterialExpressionConstantClamp_9 = Saturate(None=Multiply(A=Multiply(A=@n3, B=@n4) {constA=1}, B=@MaterialExpressionVertexColor_1.A))
  n5 = TextureSampleParameter2D(UVs=Panner(Coordinate=TextureCoordinate() {uTiling=4}) {speedX=2}) {parameterName=DetailTexture, texture=/DealershipSharedLibrary/FX/Textures/Noise/T_VCVFX_CloudsSoft_RGB}
  n6 = ScalarParameter() {defaultValue=-0.1, parameterName=MaskAmount}
  MaterialExpressionConstantClamp_11 = Saturate(None=Multiply(A=Add(A=@n5.G, B=@n6), B=ScalarParameter() {defaultValue=0.5, parameterName=Sharpness}))
  n7 = LinearInterpolate(
   A=@MaterialExpressionConstantClamp_9
   B=VectorParameter() {parameterName=Snow, defaultValue=<rgba 1 1 1 1>}.RGB
   Alpha=Multiply(A=@MaterialExpressionConstantClamp_11, B=ScalarParameter() {defaultValue=0.6, parameterName=MaskBrightness})
  )
  n8 = Panner(Coordinate=TextureCoordinate() {uTiling=4}) {speedX=1}
  n9 = Multiply(
   A=Add(
    A=TextureSample(UVs=@n8) {texture=&T_VCVFX_IceT_RGB}.R
    B=TextureSample() {texture=&T_VCVFX_IceT_RGB}.G
   )
   B=Constant() {r=0.6}
  )
  n10 = Constant3Vector() {constant=<rgba 0.6 0.7 2 0>}
  MaterialExpressionConstantClamp_8 = Saturate(None=Add(A=Add(A=@n2, B=@n7), B=Multiply(A=@n9, B=@n10)))
  n11 = ScalarParameter() {defaultValue=0.45, parameterName=Opacity}
  n12 = TextureSample(UVs=Panner(Coordinate=TextureCoordinate() {uTiling=0.5}) {speedX=0.6}) {texture=/DealershipSharedLibrary/FX/Textures/Gradient_Masks/T_VCVFX_HeroSpikes_A}
  n13 = Multiply(A=@n12.G, B=Constant() {r=1500})
  n14 = Power(Base=@n12.G) {constExponent=1}
  n15 = Constant() {r=12}
  n16 = ComponentMask(None=TextureCoordinate() {coordinateIndex=1}) {r=true}
  MaterialExpressionConstantClamp_2 = Saturate(None=Multiply(A=@n16, B=Constant() {r=15}))
  n17 = Constant() {r=1}
  n18 = Multiply(A=Power(Base=Add(A=@n13, B=@n14), Exp=Constant() {r=1}), B=Multiply(A=@n15, B=Power(Base=@MaterialExpressionConstantClamp_2, Exp=@n17)))
  n19 = OneMinus(None=@n16)
  n20 = Power(Base=Multiply(A=@n19, B=Constant() {r=4}), Exp=Constant())
  n21 = Constant() {r=1}
  n22 = Panner(Coordinate=TextureCoordinate() {vTiling=0.5}) {speedX=3, speedY=3}
  n23 = Multiply(
   A=Multiply(A=Multiply(A=@n20, B=@n21), B=TextureSample(UVs=@n22) {texture=&T_VCVFX_CloudsSoft_RGB}.G)
   B=Constant() {r=1}
  )
  MaterialExpressionConstantClamp_0 = Saturate(None=Floor(None=Add(A=@n18, B=@n23)))
  MaterialExpressionConstantClamp_13 = Saturate(None=Multiply(A=Add(A=@n11, B=@n5.G), B=@MaterialExpressionConstantClamp_0))
  n24 = MaterialFunctionCall() {materialFunction=&/MF_TyreSplash}
  MaterialFunctionCall(Emissive=LinearInterpolate(
    A=LinearInterpolate(B=@MaterialExpressionConstantClamp_8, Alpha=@MaterialExpressionConstantClamp_13)
    B=Multiply(A=@n24.Emissive) {constB=2.25}
    Alpha=@n24.Opcity
   )) {materialFunction=&MF_FX_EmissiveAdjustments}
  n25 = ComponentMask(None=TextureCoordinate()) {g=true}
  Multiply(A=Add(A=@MaterialExpressionConstantClamp_13, B=@n24.Opcity), B=Round(None=Sine(None=@n25) {period=2}))
  Multiply(A=@n19, B=Constant() {r=50})
  Multiply(A=@n19, B=@n15)
  n26 = Add(A=TextureCoordinate() {uTiling=0.65, vTiling=0.25}, B=Constant2Vector() {g=-0.13, r=-0.447})
  n27 = OneMinus(None=DotProduct(A=@n26, B=@n26))
  n28 = Constant() {r=0.8}
  MaterialExpressionConstantClamp_1 = Saturate(None=Multiply(A=Subtract(A=@n27, B=@n28), B=Constant() {r=100}))
  n29 = ComponentMask(None=TextureCoordinate() {coordinateIndex=1}) {g=true}
  n30 = Multiply(A=@n29, B=OneMinus(None=@n29))
  n31 = Constant() {r=4}
  Multiply(A=@MaterialExpressionConstantClamp_1, B=Power(Base=Multiply(A=@n30, B=@n31), Exp=Constant() {r=3}))
  n32 = Transform(None=ReflectionVectorWS()) {transformSourceType=TRANSFORMSOURCE_Local}
  MaterialExpressionConstantClamp_10 = Saturate(None=ConstantBiasScale(None=Multiply(A=@n32, B=@n3)) {bias=0.1, scale=0.2})
  Multiply(
   A=ScalarParameter() {defaultValue=4, parameterName=ReflectionAmount}
   B=Desaturation(None=@MaterialExpressionConstantClamp_10, Fraction=Constant()) {luminanceFactors=<rgba 0 0 0.11 0>}
  )
  Multiply(A=TextureSample(UVs=@n8) {texture=/DealershipSharedLibrary/FX/Textures/Diffuse/T_VCVFX_mgm_D}.A, B=Multiply(A=@n19, B=Constant() {r=30}))
  TextureSample() {samplerSource=SSM_Wrap_WorldGroupSettings, texture=/DealershipSharedLibrary/FX/Textures/Gradient_Masks/T_VCVFX_mgm_A}
 ]
}
